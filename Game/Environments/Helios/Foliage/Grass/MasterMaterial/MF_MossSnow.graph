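# Material function that tints the base color of incoming material attributes with a moss snow color, toggled by the UseMossSnow static switch.
# A static landscape snow mask, sampled by world XY relative to landscape origin and size, is contrast-remapped and multiplied by noise to drive the blend.
# Default shading combines world-space and UV-tiled noise; the mobile path uses only the UV noise, and other attributes pass through unchanged.
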MaterialFunction /Game/Environments/Helios/Foliage/Grass/MasterMaterial/MF_MossSnow
graph {nodes [
  n1 = FunctionInput() {inputType=FunctionInput_MaterialAttributes}
  n2 = GetMaterialAttributes(None=@n1) {attributeGetTypes [#1]}
  n3 = Subtract(A=WorldPosition().XYZ, B=VectorParameter() {parameterName=CoordsOrgin, defaultValue=<rgba -130048 -130048 100 1>}.RGB)
  n4 = VectorParameter() {parameterName=LandscapeSize, defaultValue=<rgba 260096 260096 100 1>}
  n5 = TextureSample(UVs=ComponentMask(None=Divide(A=@n3, B=@n4.RGB)) {g=true, r=true}) {
   samplerSource=SSM_Wrap_WorldGroupSettings
   samplerType=SAMPLERTYPE_LinearColor
   texture=/Game/Athena/Apollo/Environments/Landscape/Materials/Biome/CommonMaterials/T_SnowMask_Static
  }
  n6 = Saturate(None=LinearInterpolate(
    A=ScalarParameter() {defaultValue=-1.6376, group=Snow, parameterName=SnowWorldBiomeMask_Contrast_Low}
    B=ScalarParameter() {defaultValue=2.9577, group=Snow, parameterName=SnowWorldBiomeMask_Contrast_High}
    Alpha=@n5.R
   ))
  n7 = TextureSample(UVs=Divide(A=WorldPosition().XY, B=ScalarParameter() {defaultValue=621.4855, group=Snow, parameterName=SnowNoiseSizeWS})) {samplerSource=SSM_Wrap_WorldGroupSettings, samplerType=SAMPLERTYPE_Masks, texture=/Game/Global/Textures/Default/Noise/T_Metal_Noise}
  n8 = ScalarParameter() {defaultValue=0.5368, group=Snow, parameterName=SnowNoiseContrast2}
  n9 = TextureSample(UVs=Multiply(A=TextureCoordinate(), B=ScalarParameter() {defaultValue=2.7568, group=Snow, parameterName=SnowNoiseTiling})) {samplerSource=SSM_Wrap_WorldGroupSettings, samplerType=SAMPLERTYPE_Masks, texture=&T_Metal_Noise}
  n10 = Power(Base=@n9.R, Exp=ScalarParameter() {defaultValue=0.7413, group=Snow, parameterName=SnowNoiseContrast})
  n11 = ShadingPathSwitch(Default=Multiply(A=Power(Base=@n7.R, Exp=@n8), B=@n10), Mobile=@n10) {desc=Migrated from FeatureLevelSwitch}
  n12 = LinearInterpolate(
   A=@n2.BaseColor
   B=VectorParameter() {group=Snow, parameterName=SnowColor_Moss, defaultValue=<rgba 0.8385 0.8385 0.8385 1>}.RGB
   Alpha=Saturate(None=Multiply(A=@n6, B=@n11))
  )
  FunctionOutput(None=StaticSwitchParameter(
    True=SetMaterialAttributes!(MaterialAttributes=@n2.MaterialAttributes, BaseColor=@n12) {attributeSetTypes [#1]}
    False=@n1
   ) {parameterName=UseMossSnow})
 ]}
