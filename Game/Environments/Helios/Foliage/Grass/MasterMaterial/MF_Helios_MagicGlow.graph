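# Material function for Helios grass that layers an optional "magic glow" tint onto incoming material attributes, toggled by the UseHeliosMagicGlow static switch.
# The Wind input drives a biased mask, while transformed normals dotted with a camera vector, its Z replaced by a parameter, set tint strength.
# The tint adds to BaseColor, not Emissive, and the mask also raises Specular and Metallic.
# Pivot-based sine wave and power nodes never reach the output.
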MaterialFunction /Game/Environments/Helios/Foliage/Grass/MasterMaterial/MF_Helios_MagicGlow
graph {nodes [
  n1 = FunctionInput() {bUsePreviewValueAsDefault=true, inputName=Pivot Posistion, sortPriority=10}
  ComponentMask(None=@n1) {g=true}
  FunctionInput() {bUsePreviewValueAsDefault=true, inputName=WindDirection, sortPriority=11}
  n2 = FunctionInput() {inputType=FunctionInput_MaterialAttributes}
  n3 = GetMaterialAttributes(None=@n2) {attributeGetTypes [#1, #2, #3]}
  n4 = VectorParameter() {group=MagicGlow, parameterName=MagicGlow, defaultValue=<rgba 1 0.7093 0.2292 1>}
  n5 = FunctionInput() {bUsePreviewValueAsDefault=true, inputName=Normals, sortPriority=7}
  n6 = ComponentMask(None=CameraVectorWS()) {g=true, r=true}
  n7 = ScalarParameter() {defaultValue=15.0889, group=MagicGlow, parameterName=MagicGlow_DownVector}
  n8 = DotProduct(A=Normalize(VectorInput=Transform(None=@n5)), B=Normalize(VectorInput=AppendVector(A=@n6, B=@n7)))
  n9 = LinearInterpolate(
   A=ScalarParameter() {group=MagicGlow, parameterName=MagicGlow_Power_low}
   B=ScalarParameter() {defaultValue=0.8619, group=MagicGlow, parameterName=MagicGlow_Power_High}
   Alpha=Add(A=@n8)
  )
  n10 = Saturate(None=LinearInterpolate(
    A=ScalarParameter() {defaultValue=-0.2, group=MagicGlow, parameterName=MagicGlow_Mask_Low}
    B=ScalarParameter() {defaultValue=0.7511, group=MagicGlow, parameterName=MagicGlow_Mask_High}
    Alpha=FunctionInput() {bUsePreviewValueAsDefault=true, inputName=Wind, inputType=FunctionInput_Scalar, sortPriority=5}
   ))
  n11 = Add(A=@n10, B=ScalarParameter() {defaultValue=0.1, parameterName=MagicGlow_Bias})
  n12 = Add(A=@n3.BaseColor, B=Multiply(A=Multiply(A=@n4.RGB, B=@n9), B=@n11))
  n13 = Add(A=@n3.Specular, B=Multiply(A=ScalarParameter() {defaultValue=1, group=MagicGlow, parameterName=MagicGlow_Specular}, B=@n11))
  n14 = LinearInterpolate(A=@n3.Metallic, B=ScalarParameter() {defaultValue=1, group=MagicGlow, parameterName=MagicGlow_Metallic}, Alpha=@n11)
  FunctionOutput(None=StaticSwitchParameter(
    True=SetMaterialAttributes!(MaterialAttributes=@n3.MaterialAttributes, BaseColor=@n12, Specular=@n13, Metallic=@n14) {attributeSetTypes [#1, #2, #3]}
    False=@n2
   ) {group=MagicGlow, parameterName=UseHeliosMagicGlow})
  Multiply()
  Power(Base=@n8, Exp=ScalarParameter() {defaultValue=-0.094, group=MagicGlow, parameterName=MagicGlow_Power})
  n15 = DotProduct(A=ComponentMask(None=@n1) {g=true, r=true}, B=Constant3Vector() {constant=<rgba 1 1 0 0>})
  n16 = ScalarParameter() {defaultValue=889.819, group=MagicGlow, parameterName=MagicGlow_Wave_Length}
  n17 = Time()
  n18 = ScalarParameter() {defaultValue=0.5, group=MagicGlow, parameterName=MagicGlow_Wave_speed}
  n19 = Sine(None=Add(A=Divide(A=@n15, B=@n16), B=Multiply(A=@n17, B=@n18)))
  StaticSwitch(True=Divide(A=Add(A=@n19)), False=@n10)
 ]}
