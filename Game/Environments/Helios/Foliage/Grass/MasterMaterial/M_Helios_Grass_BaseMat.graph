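# Master material for Helios grass foliage: two-sided, Nanite-capable, with static switches for default, GrassV2 and Scotch Broom variants.
# The chosen variant passes through magic glow, optional snow and landscape functions.
# Color mixes textures, per-instance hue variation and landscape biome/RVT blending, with cheaper mobile and low-quality paths.
# World position offset builds wind bending, player push and physics-field motion from baked UV1 pivots and vertex-color blade masks.
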Material /Game/Environments/Helios/Foliage/Grass/MasterMaterial/M_Helios_Grass_BaseMat
graph {
 properties {
  bTangentSpaceNormal=false
  bUseMaterialAttributes=true
  bUsedWithInstancedStaticMeshes=true
  bUsedWithNanite=true
  pixelDepthOffsetMode=PDOM_Legacy
  shadingModel=MSM_FromMaterialExpression
  twoSided=true
  usedShadingModels=Default Lit | Subsurface
 }
 nodes [
  Constant3Vector() {constant=<rgba 0 0 100 0>}
  Constant4Vector()
  Constant() {r=1.25}
  Constant() {r=1}
  Constant() {r=1}
  Constant() {r=0.35}
  Constant() {r=0.4}
  Constant() {r=2}
  Constant() {r=-7000}
  Constant() {r=5}
  Constant() {r=0.1}
  Constant() {r=0.5}
  Constant() {r=0.85}
  Constant() {r=0.85}
  Constant() {r=6000}
  n1 = NamedRerouteUsage()
  n2 = NamedRerouteUsage()
  n3 = MaterialFunctionCall!(
   In=StaticSwitchParameter(
    True=NamedRerouteUsage()
    False=StaticSwitchParameter(True=@n1, False=@n2) {desc="This should only be used for the simple, bladed grass", group=GrassTypeSettings, parameterName=UseGrassV2}
   ) {group=GrassTypeSettings_SB, parameterName=UseSB}
   Wind=NamedRerouteUsage()
   Normals=NamedRerouteUsage()
   Pivot Posistion=NamedRerouteUsage()
   WindDirection=NamedRerouteUsage()
  ) {materialFunction=&/MF_Helios_MagicGlow}
  n4 = GetMaterialAttributes(None=@n3.Result) {attributeGetTypes [#1]}
  n5 = ConstantBiasScale(None=DotProduct(A=Constant3Vector() {constant=<rgba 0 0 1 0>}, B=VertexNormalWS()))
  n6 = ScalarParameter() {defaultValue=1, group=Snow, parameterName=G_SnowSample}
  n7 = ScalarParameter() {defaultValue=0.3, group=Snow, parameterName=G_SnowSampleRadius}
  n8 = ScalarParameter() {defaultValue=0.5, group=Snow, parameterName=G_SnowSampleHardness}
  n9 = LinearInterpolate(
   A=@n4.Base Color
   B=VectorParameter() {group=Snow, parameterName=G_SnowColor, defaultValue=<rgba 0.4292 0.4575 0.5 1>}.RGB
   Alpha=Multiply(
    A=MaterialFunctionCall!(
     ScalarWeightMask
     UseSnowOcclusionTextureMask
     ------
     UseMaterialInput
     AdditionalMaskToRemoveSnow
     OutputTangentNormals
     UseAdditionalMaskToRemoveSnow
     UseSnowNormals
     OuputVertexNormalsAsTheSnowNormals
     UseSparkle
     LQSpark
     AllowSnowMicroNormals
     CliffMaterialOnly
     Material
     CliffLandscapeLayerMask
     IsLandscape
     AreInputNormalsTangentSpace
     UseInputNormalsForFakeSnowShadow
    ) {materialFunction=/Game/Athena/Helios/Environments/Landscape/Materials/QualityLevelLayers/MF_Helios_Snow_Default}.DetailMask
    B=SphereMask(A=@n5, B=@n6, Radius=@n7, Hardness=@n8)
   )
  )
  n10 = MaterialFunctionCall!(
   In=StaticSwitchParameter(
    True=SetMaterialAttributes!(MaterialAttributes=@n4.MaterialAttributes, Base Color=@n9) {attributeSetTypes [#1]}
    False=@n3.Result
   ) {group=Snow, parameterName=UseSnow, sortPriority=-100}
   UseHeightFalloff=StaticBool()
   OnFireEmissiveOverride
   HeightFalloffRadius
   HeightFalloffHardness
   VariationMasksMobile
   In_Grass=StaticBool() {value=true}
   Is_Stylised_WE
  ) {materialFunction=/Game/Packages/Fortress_SharedMaterials/Base_Material_Functions/MF_Curie_Landscape}
  GetMaterialAttributes(None=@n10.Result) {attributeGetTypes [#1]}
  n11 = NamedRerouteUsage()
  LinearInterpolate(
   A=ScalarParameter() {group=Color, parameterName=HueVarPerBladeA}
   B=ScalarParameter() {defaultValue=0.05, group=Color, parameterName=HueVarPerBladeB}
   Alpha=@n11
  )
  n12 = VectorParameter() {
   group=GrassTypeSettings
   parameterName=ColorTint2
   sortPriority=-499
   channelNames {a=Desaturate, b="", g="", r=""}
   defaultValue=<rgba 1 1 1 0>
  }
  n13 = NamedRerouteUsage()
  n14 = Multiply(A=@n12, B=Desaturation(None=@n13, Fraction=@n12.Desaturate))
  n15 = Saturate(None=Multiply(A=@n13, B=VectorParameter() {group=GrassTypeSettings, parameterName=ColorTint, sortPriority=-500, defaultValue=<rgba 1 1 1 1>}.RGB))
  n16 = PerInstanceRandom()
  n17 = ScalarParameter() {group=Color, parameterName=HueVarPerClusterA}
  n18 = ScalarParameter() {defaultValue=0.05, group=Color, parameterName=HueVarPerClusterB}
  n19 = StaticSwitchParameter(
   True=LinearInterpolate(A=Saturate(None=@n14), B=@n15, Alpha=@n16)
   False=MaterialFunctionCall(Hue Shift Percentage=LinearInterpolate(A=@n17, B=@n18, Alpha=@n16), Texture=@n15) {materialFunction=&HueShift}.Result
  ) {
   desc=False will use HueShift PerInstanceRandom instead
   group=GrassTypeSettings
   parameterName=UseColorTintPerInstanceRandom
   sortPriority=-501
  }
  n20 = NamedRerouteDeclaration(None=@n19) {name=HueShiftedBaseColor_ONLY, nodeColor=<rgba 0 1 0.5078 1>}
  n21 = LinearInterpolate(
   A=ScalarParameter() {group=GrassTypeSettings_MaskedColorVariation, parameterName=MaskedColorVariation_PerClusterA}
   B=ScalarParameter() {group=GrassTypeSettings_MaskedColorVariation, parameterName=MaskedColorVariation_PerClusterB}
   Alpha=PerInstanceRandom()
  )
  n22 = VectorParameter() {group=GrassTypeSettings_MaskedColorVariation, parameterName=MaskedColorVariation, defaultValue=<rgba 1 1 1 1>}
  n23 = Saturate(None=Multiply(A=NamedRerouteUsage(), B=@n22.RGB))
  n24 = Add(A=Multiply(A=NamedRerouteUsage()) {constB=3}) {constB=-2}
  n25 = StaticSwitchParameter(
   True=LinearInterpolate(
    A=@n20
    B=MaterialFunctionCall(Hue Shift Percentage=@n21, Texture=@n23) {materialFunction=&HueShift}.Result
    Alpha=Saturate(None=@n24)
   )
   False=@n20
  ) {group=GrassTypeSettings_MaskedColorVariation, parameterName=UseMaskedColorVariation}
  LinearInterpolate(A=@n25, B=NamedRerouteUsage(), Alpha=ScalarParameter() {group=GrassTypeSettings, parameterName=RVT/BiomeColorInfluenceonDiffuse})
  n26 = NamedRerouteUsage()
  n27 = VectorParameter() {desc="3,3,1 looks good", group=GrassTypeSettings_SB, parameterName=SB_NormalMod, defaultValue=<rgba 1 1 1 1>}
  n28 = Multiply(A=NamedRerouteUsage(), B=ComponentMask(None=@n27.RGB) {b=true, g=true, r=true})
  n29 = LinearInterpolate(A=Multiply(A=Constant3Vector() {constant=<rgba -1 -1 1 0>}, B=@n28), B=@n28, Alpha=NamedRerouteUsage())
  LinearInterpolate(A=@n26, B=LinearInterpolate(A=@n26, B=Transform(None=@n29), Alpha=Constant() {r=0.1}), Alpha=NamedRerouteUsage())
  n30 = StaticSwitchParameter(True=NamedRerouteUsage(), False=NamedRerouteUsage()) {group=GrassTypeSettings, parameterName=IsFallbackMaterial}
  n31 = Multiply(A=Constant3Vector() {constant=<rgba -1 -1 1 0>}, B=@n30)
  n32 = NamedRerouteUsage()
  n33 = NamedRerouteUsage()
  n34 = ComponentMask(None=VectorParameter() {group=GrassTypeSettings, parameterName=GrassV2_NormalAmountT, defaultValue=<rgba 0.5 0.5 0.5 1>}.RGB) {b=true, g=true, r=true}
  n35 = Transform(None=Add(A=Transform(None=@n33) {transformSourceType=TRANSFORMSOURCE_Instance, transformType=TRANSFORM_Tangent}, B=Multiply(A=@n30, B=@n34)))
  LinearInterpolate(
   A=Transform(None=LinearInterpolate(A=@n31, B=@n30, Alpha=@n32))
   B=@n35
   Alpha=ScalarParameter() {group=GrassTypeSettings, parameterName=GrassV2_UseVertexNormalsInsteadOfNormalsTexture}
  )
  n36 = StaticSwitchParameter(True=NamedRerouteUsage(), False=NamedRerouteUsage()) {parameterName=UseCameraRelativeOptimization}
  n37 = CollectionParameter() {collection=&FortniteMaterialParameters, parameterName=PlayerPosition}
  n38 = ComponentMask(None=@n37) {b=true, g=true, r=true}
  n39 = StaticSwitchParameter(
   True=TransformPosition(None=@n38) {transformSourceType=TRANSFORMPOSSOURCE_World, transformType=TRANSFORMPOSSOURCE_TranslatedWorld}
   False=@n38
  ) {parameterName=UseCameraRelativeOptimization}
  n40 = Normalize(VectorInput=Subtract(A=@n36, B=@n39))
  n41 = Saturate(None=SphereMask(A=@n39, B=@n36, Radius=Constant() {r=200}, Hardness=Constant()) {attenuationRadius=64, hardnessPercent=20}) {desc=PlayerSphereMask}
  LinearInterpolate(
   A=NamedRerouteUsage()
   B=Multiply(A=VectorParameter() {group=GrassTypeSettings_WPO, parameterName=PushDistance, defaultValue=<rgba -22.5 -22.5 -50 1>}.RGB, B=@n40)
   Alpha=@n41
  )
  MaterialFunctionCall!(BiomeMask-RGBA, TextureSamplingPosition=NamedRerouteUsage(), IsGrass=StaticBool() {value=true}) {materialFunction=/Game/Athena/Asteria/Environments/Landscape/Materials/MaterialFunctions/MF_AsteriaGrassColors}
  n42 = ScalarParameter() {
   defaultValue=0.2
   desc=only blend To RVT at the bottom So you can have green grass on sand
   group=GrassTypeSettings
   parameterName=NonRVTBlend
  }
  MaterialFunctionCall(In=NamedRerouteUsage(), Contrast=@n42) {materialFunction=&CheapContrast}
  MaterialFunctionCall!(
   Previous WPO Value
   Force Magnitude Mask
   Force Magnitude Multiplier
   Project Field Force Vector to Normals
   Noise Size
   Noise Pan Speed
   Noise Falloff
   Stretch Noise Coordinates Z
  ) {desc=Sample global physics field add to WPO, materialFunction=/Game/Effects/Niagara/MasterMaterials/Functions/MF_FX_PhysicsFields_Grass}
  n43 = MaterialFunctionCall!(UseLandscapeCoords, "UseProjection(translucent)") {materialFunction=/Game/Athena/Helios/Environments/Landscape/Materials/MaterialFunctions_Utilities/MF_TerrainMasks}
  n44 = MaterialFunctionCall!(
   TerrainMasks_QuadA=@n43.QuadA_RGBA
   TerrainMasks_QuadB=@n43.QuadA_RGBA
   TerrainMasks_QuadC=@n43.QuadA_RGBA
   TerrainMasks_QuadD=@n43.QuadA_RGBA
   VariationMasks Mobile=MaterialFunctionCall() {materialFunction=/Game/Athena/Helios/Environments/Landscape/Materials/MaterialFunctions_Utilities/MF_VariationMasks_Mobile}.VariationMasks Mobile
  ) {materialFunction=/Game/Athena/Helios/Environments/Landscape/Materials/QualityLevelLayers/MF_Helios_Grass_NS}
  MaterialFunctionCall!(GrassMaterial=@n44.Grass Foliage Blend, Layer Mask, Frensel Color) {materialFunction=/Game/Athena/Helios/Environments/Landscape/Materials/MaterialFunctions_Utilities/MF_MobileGrass_Fresnel}
  n45 = LinearInterpolate(
   A=VectorParameter() {group=GrassTypeSettings_SB, parameterName=SB_Color_StalkA, defaultValue=<rgba 0.3889 0.5 0.25 1>}.RGB
   B=VectorParameter() {group=GrassTypeSettings_SB, parameterName=SB_Color_StalkB, defaultValue=<rgba 0.2467 0.3 0.18 1>}.RGB
   Alpha=PerInstanceRandom()
  )
  n46 = VectorParameter() {group=GrassTypeSettings_SB, parameterName=SB_Color_NoFlower, defaultValue=<rgba 0.4444 0.5 0.375 1>}
  n47 = Multiply(A=@n45, B=@n46.RGB)
  n48 = StaticSwitchParameter(True=Constant() {r=1}, False=NamedRerouteUsage()) {group=GrassTypeSettings, parameterName=IsFallbackMaterial}
  Multiply(A=@n47, B=Saturate(None=Add(A=@n46.A, B=@n48)))
  n49 = NamedRerouteUsage()
  n50 = LinearInterpolate(A=ScalarParameter() {defaultValue=0.5, group=GrassTypeSettings_WPO, parameterName=AnimBendFalloff_min}, Alpha=NamedRerouteUsage())
  n51 = ScalarParameter() {defaultValue=5, group=GrassTypeSettings_WPO, parameterName=AnimBendFalloff_Exponent}
  n52 = Multiply(
   A=Multiply(A=NamedRerouteUsage(), B=Power(Base=@n50, Exp=@n51))
   B=ScalarParameter() {defaultValue=150, group=GrassTypeSettings_WPO, parameterName=AnimStrength_Global}
  )
  n53 = StaticSwitchParameter(True=WorldPosition() {worldPositionShaderOffset=WPT_CameraRelativeNoOffsets}.XYZ, False=LocalPosition() {includedOffsets=ExcludeOffsets}.XYZ) {parameterName=UseCameraRelativeOptimization}
  n54 = StaticSwitchParameter(True=NamedRerouteUsage(), False=NamedRerouteUsage()) {parameterName=UseCameraRelativeOptimization}
  n55 = Subtract(A=Add(A=@n53, B=Multiply(A=@n49, B=@n52)), B=@n54)
  n56 = Length(None=@n55) {is=class MaterialExpressionLength}
  n57 = Length(None=Subtract(A=@n53, B=@n54))
  Multiply(A=@n55, B=Divide(A=@n56, B=@n57))
  n58 = Constant3Vector() {constant=<rgba 0 0 1 0>}
  NamedRerouteDeclaration(None=VertexInterpolator(VS=Transform(None=@n58) {transformSourceType=TRANSFORMSOURCE_Instance}).PS) {name=Instance_Z_To_World, nodeColor=<rgba 0 0 1 1>}
  n59 = TextureSampleParameter2D() {
   group=GrassTypeSettings_Textures
   parameterName=NormalTexture
   samplerSource=SSM_Wrap_WorldGroupSettings
   samplerType=SAMPLERTYPE_Normal
   sortPriority=5
   texture=/Game/Environments/Asteria/Foliage/Grass/Grass2DAutoBake/Grass_Standard_2D_N
  }
  NamedRerouteDeclaration(None=MaterialFunctionCall(Normal=@n59.RGB, Flatness=ScalarParameter() {group=GrassTypeSettings, parameterName=BakedTextureNormalFlatten}) {materialFunction=&FlattenNormal}.Result) {name=BakedNormalTexture, nodeColor=<rgba 1 0.9375 0 1>}
  n60 = TextureCoordinate() {coordinateIndex=1}
  n61 = MaterialFunctionCall(X=ComponentMask(None=@n60) {r=true}, Y=ComponentMask(None=@n60) {g=true}, Z=Constant()) {materialFunction=&MakeFloat3}
  NamedRerouteDeclaration(None=TransformPosition(None=@n61.Result) {transformSourceType=TRANSFORMPOSSOURCE_Instance, transformType=TRANSFORMPOSSOURCE_TranslatedWorld}) {name=PivotPosition_CameraRelative, nodeColor=<rgba 0 0.9766 1 1>}
  NamedRerouteDeclaration(None=LinearInterpolate(
    A=ScalarParameter() {defaultValue=1, group=GrassTypeSettings, parameterName=GrassV2_SSOpacity_Far}
    B=ScalarParameter() {defaultValue=0.25, group=GrassTypeSettings, parameterName=GrassV2_SSOpacity_Near}
    Alpha=NamedRerouteUsage()
   )) {name=GrassV2_Opacity, nodeColor=<rgba 1 0 1 1>}
  n62 = ScalarParameter() {defaultValue=0.1, group=GrassTypeSettings, parameterName=GrassV2_Spec1}
  n63 = ScalarParameter() {defaultValue=0.5, group=GrassTypeSettings, parameterName=GrassV2_Spec2}
  n64 = NamedRerouteUsage()
  NamedRerouteDeclaration(None=Multiply(A=LinearInterpolate(A=@n62, B=@n63, Alpha=@n64) {constA=0.3, constB=1.2}, B=NamedRerouteUsage())) {name=GrassV2_Specular, nodeColor=<rgba 1 0 1 1>}
  NamedRerouteDeclaration(None=Multiply(A=TwoSidedSign(), B=Transform(None=@n58) {transformSourceType=TRANSFORMSOURCE_Instance, transformType=TRANSFORM_Tangent})) {name=InstanceZ_NormalsTangent, nodeColor=<rgba 1 0 1 1>}
  n65 = MaterialFunctionCall() {materialFunction=/Game/Athena/Helios/Environments/Landscape/Materials/MaterialFunctions_Utilities/MF_ScotchBroom_Colors}
  n66 = LinearInterpolate(A=@n47, B=@n65.Flower1_Color, Alpha=@n65.Flower1_Color)
  n67 = NamedRerouteUsage()
  n68 = MaterialFunctionCall(Float3=NamedRerouteUsage()) {materialFunction=&BreakOutFloat3Components}
  n69 = Min(A=Min(A=@n68.R, B=@n68.R), B=@n68.B)
  n70 = Add(A=Max(A=Max(A=@n68.R, B=@n68.R), B=@n68.B), B=@n69)
  n71 = LinearInterpolate(
   A=ScalarParameter() {defaultValue=-1, group=GrassTypeSettings_SB, parameterName=SB_MP Mask LerpA}
   B=ScalarParameter() {defaultValue=2, group=GrassTypeSettings_SB, parameterName=SB_MP Mask LerpB}
   Alpha=Saturate(None=Multiply(A=@n70, B=@n69))
  )
  n72 = Multiply(A=@n65.FlowerWorldMask1, B=NamedRerouteDeclaration(None=Saturate(None=@n71)) {name=SB_Fmask, nodeColor=<rgba 1 0 1 1>})
  n73 = ScalarParameter() {defaultValue=0.85, group=GrassTypeSettings_SB, parameterName=SB_Flat_Roughness}
  n74 = OneMinus(None=NamedRerouteUsage())
  n75 = LinearInterpolate(A=@n73, B=Saturate(None=Add(A=@n73, B=@n74)), Alpha=NamedRerouteUsage()) {constA=0.3, constB=1.2}
  n76 = NamedRerouteUsage()
  n77 = ScalarParameter() {defaultValue=900, group=GrassTypeSettings_SB, parameterName=SB_ChaparralCoast_GrassScale}
  n78 = MaterialFunctionCall!(
   Diffuse Texture Object
   Spec Texture Object
   TextureSize=@n77
   --------------------------
   Use seperate settings for Z projection
   Use seperate texture for Z projection
   Texture Size Z projection=@n77
   Z Plane Diffuse Texture Object
   Z Plane Spec Texture Object
   "--------------------------      "
   ProjectionTransitionContrast
   Use Texture Alpha Heightmap to Refine the Transistion
   HeightLerpTransitionContrast
   PickTheDiffuseHeightChannel
   TheDiffuseHeightChannel
   "--------------------------       "
   NormalTextureObject=TextureObject() {samplerType=SAMPLERTYPE_Normal, texture=/Game/Environments/Helios/Terrain/Textures/T_Helios_Chap_CoastGround_01_N}
   World Space Normals=StaticBool() {value=true}
   "--------------------------  "
   Export Float 4 Diffuse
   Export Float 4 Spec
   "--------------------------     "
   WorldPosition=WorldPosition() {worldPositionShaderOffset=WPT_ExcludeAllShaderOffsets}.XYZ
   World Space Normal=NamedRerouteUsage()
  ) {materialFunction=/Engine/Functions/Engine_MaterialFunctions02/Texturing/WorldAlignedTextures_Complex}
  n79 = NamedRerouteUsage()
  n80 = Constant3Vector() {constant=<rgba 0.3 0.75 0.1 0>}
  n81 = LinearInterpolate(A=@n80.R, B=@n80.R, Alpha=NamedRerouteUsage())
  n82 = MaterialFunctionCall(X=@n81, Y=@n81, Z=@n80.B) {materialFunction=&MakeFloat3}
  n83 = NamedRerouteUsage()
  n84 = Transform(None=Add(A=Transform(None=@n79) {transformSourceType=TRANSFORMSOURCE_Instance, transformType=TRANSFORM_Tangent}, B=Multiply(A=@n82.Result, B=@n83)))
  n85 = NamedRerouteUsage()
  n86 = SetMaterialAttributes!(
   MaterialAttributes
   Base Color=Multiply(A=LinearInterpolate(A=@n45, B=@n66, Alpha=@n67) {desc=Basecolor}, B=NamedRerouteUsage())
   Specular=LinearInterpolate(
    A=ScalarParameter() {defaultValue=0.6, desc=.6, group=GrassTypeSettings_SB, parameterName=SB_Flat_SpecFar}
    B=ScalarParameter() {defaultValue=0.05, desc=.05, group=GrassTypeSettings_SB, parameterName=SB_Flat_SpecNear}
    Alpha=NamedRerouteUsage()
   ) {constA=0.3, constB=1.2}
   Roughness=Saturate(None=Add(A=@n72, B=@n75))
   Opacity=NamedRerouteUsage()
   Opacity Mask=@n76
   Normal=ShadingPathSwitch(Default=LinearInterpolate(A=@n78.Normal Z Texture, B=@n84, Alpha=@n85), Mobile=NamedRerouteUsage())
   Subsurface Color=Constant()
   World Position Offset=NamedRerouteUsage()
   Shading Model=ShadingModel() {shadingModel=MSM_Subsurface}
  ) {attributeSetTypes [
    #1
    #2
    #3
    #4
    #5
    #6
    #7
    #8
    #9
   ]}
  NamedRerouteDeclaration(None=@n86) {name=SB, desc=ScotchBroom, nodeColor=<rgba 1 0 0.1406 1>}
  n87 = NamedRerouteUsage()
  n88 = ScalarParameter() {defaultValue=0.5, group=GrassTypeSettings, parameterName=SoftNormal_RoughnessAdd}
  n89 = SetMaterialAttributes!(
   MaterialAttributes
   Base Color=NamedRerouteUsage()
   Specular=Multiply(A=NamedRerouteUsage(), B=ScalarParameter() {defaultValue=0.5, group=GrassTypeSettings, parameterName=SoftNormal_SpecAmount})
   Roughness=Saturate(None=Add(A=@n87, B=@n88))
   Opacity=NamedRerouteUsage()
   Normal=NamedRerouteUsage()
   World Position Offset=NamedRerouteUsage()
   Subsurface Color=NamedRerouteUsage()
   Opacity Mask=NamedRerouteUsage()
   Shading Model=ShadingModel() {shadingModel=MSM_Subsurface}
  ) {attributeSetTypes [
    #1
    #2
    #3
    #4
    #6
    #8
    #7
    #5
    #9
   ]}
  NamedRerouteDeclaration(None=@n89) {name=DefaultBase_Material, nodeColor=<rgba 1 0 1 1>}
  NamedRerouteDeclaration(None=SphereMask(
    A=CameraPositionWS()
    B=NamedRerouteUsage()
    Radius=ScalarParameter() {defaultValue=6000, group=GrassTypeSettings, parameterName=GrassV2_PerInstanceNormal_Distance}
    Hardness=Constant()
   )) {name=NormalsDistanceMask, nodeColor=<rgba 1 0 1 1>}
  NamedRerouteDeclaration(None=@n58) {name=001, nodeColor=<rgba 0 0 1 1>}
  n90 = NamedRerouteUsage()
  n91 = ScalarParameter() {defaultValue=0.35, group=GrassTypeSettings, parameterName=GrassV2_AmountOfNormalTypeNearCamera}
  NamedRerouteDeclaration(None=LinearInterpolate(A=@n90, B=LinearInterpolate(A=@n90, B=@n35, Alpha=@n91), Alpha=NamedRerouteUsage())) {name=GrassV2_Normals, nodeColor=<rgba 1 0 1 1>}
  n92 = TextureSampleParameter2D() {
   group=GrassTypeSettings_Textures
   parameterName=DiffuseTexture
   samplerSource=SSM_Wrap_WorldGroupSettings
   sortPriority=0
   texture=/Game/Characters/Fortress_Character/Textures/T_All_White
  }
  NamedRerouteDeclaration(None=Desaturation(None=@n92.RGB, Fraction=ScalarParameter() {group=GrassTypeSettings, parameterName=DiffuseTextureDesaturation})) {name=DiffuseTexture, nodeColor=<rgba 1 0.3984 0 1>}
  n93 = NamedRerouteUsage()
  n94 = StaticSwitchParameter(True=@n93, False=WorldPosition().XYZ) {defaultValue=true, group=GrassTypeSettings_WPO, parameterName=UseSamplingPerBladeOrWorldPosition}
  n95 = Multiply(A=NamedRerouteUsage(), B=ScalarParameter() {defaultValue=200, group=GrassTypeSettings_WPO, parameterName=RandomSampleOffsetPerBlade})
  n96 = MaterialFunctionCall!(
   Position=StaticSwitchParameter(True=@n94, False=Add(A=@n95, B=@n94)) {group=GrassTypeSettings_WPO, parameterName="UseRealRotationAnimation(MostExpensive)"}
   IsTrees
   IsGrass=StaticBool() {value=true}
   IsMediumPlant
   IsMediumPlant_2024Anim
   IsMediumPlant2ndLevel
   MediumPlant_BasePivotPostion
  ) {materialFunction=/Game/Environments/Asteria/Foliage/Wind/MF_WindActorData}
  NamedRerouteDeclaration(None=@n96.WindIntensitySingleSample) {name=WindIntensityNoise, nodeColor=<rgba 0.1719 0 1 1>}
  n97 = TransformPosition(None=@n61.Result) {transformSourceType=TRANSFORMPOSSOURCE_Instance, transformType=TRANSFORMPOSSOURCE_World}
  NamedRerouteDeclaration(None=@n97) {name=PivotPosition, nodeColor=<rgba 0 0.9766 1 1>}
  n98 = PixelDepth()
  n99 = Add(A=@n98.R, B=ScalarParameter() {defaultValue=-2000, group=GrassTypeSettings, parameterName=NearFadeOffset})
  n100 = ScalarParameter() {defaultValue=5000, group=GrassTypeSettings, parameterName=NearFade}
  NamedRerouteDeclaration(None=Saturate(None=Divide(A=@n99, B=@n100))) {name=NearFadeMask, nodeColor=<rgba 0.3802 0.3802 0.3802 1>}
  n101 = Add(A=@n98.R, B=ScalarParameter() {defaultValue=-7000, group=GrassTypeSettings, parameterName=FarFadeOffset})
  n102 = ScalarParameter() {defaultValue=4000, group=GrassTypeSettings, parameterName=FarFade}
  NamedRerouteDeclaration(None=Saturate(None=Divide(A=@n101, B=@n102))) {name=FarFadeMask, nodeColor=<rgba 0.3802 0.3802 0.3802 1>}
  n103 = GetMaterialAttributes(None=@n44.Grass Foliage Blend) {attributeGetTypes [#1]}
  n104 = StaticSwitchParameter(True=Multiply(A=NamedRerouteUsage(), B=@n103.Base Color), False=@n103.Base Color) {parameterName=UseDiffuseColor}
  n105 = VectorParameter() {group=GrassTypeSettings_Mobile, parameterName=Mobile Blades 1, defaultValue=<rgba 1.2015 1.2867 1.5 1>}
  n106 = VectorParameter() {group=GrassTypeSettings_Mobile, parameterName=Mobile Blades 2, defaultValue=<rgba 1.2 1.1592 1.1125 1>}
  n107 = LinearInterpolate(
   A=@n104
   B=LinearInterpolate(A=Multiply(A=@n105.RGB, B=@n104) {constB=1.1}, B=Multiply(A=@n106.RGB, B=@n104) {constB=1.1}, Alpha=NamedRerouteUsage()) {constAlpha=0.1}
   Alpha=NamedRerouteUsage()
  )
  n108 = NamedRerouteUsage()
  n109 = NamedRerouteUsage()
  n110 = MaterialFunctionCall(Float3=@n109) {materialFunction=&BreakOutFloat3Components}
  n111 = NamedRerouteUsage()
  n112 = Saturate(None=Multiply(A=LinearInterpolate(A=@n110.R, B=@n110.R, Alpha=@n111)) {constB=-1})
  n113 = NamedRerouteUsage()
  n114 = Add(
   A=Multiply(A=Multiply(A=@n112, B=@n113))
   B=ScalarParameter() {defaultValue=0.1, group=GrassTypeSettings_Mobile, parameterName=Mobile Specular}
  )
  n115 = NamedRerouteUsage()
  n116 = Constant3Vector() {desc=Looks almost same RVT normal, constant=<rgba 0 0 1 0>}
  n117 = Transform(None=@n116) {transformSourceType=TRANSFORMSOURCE_Instance}
  n118 = VertexInterpolator(VS=@n117)
  n119 = Multiply(A=Constant3Vector() {constant=<rgba 1 1 0.5 0>}, B=@n117) {constB=-1}
  n120 = LinearInterpolate(A=VertexInterpolator(VS=Normalize(VectorInput=@n119)).PS, B=LinearInterpolate(A=@n116, B=@n109), Alpha=NamedRerouteUsage()) {constAlpha=1}
  n121 = Multiply(A=NamedRerouteUsage()) {constB=0.2}
  n122 = SetMaterialAttributes!(
   MaterialAttributes
   Base Color=Multiply(A=Multiply(A=@n107, B=@n108), B=NamedRerouteUsage())
   Roughness=ScalarParameter() {defaultValue=0.55, group=GrassTypeSettings_Mobile, parameterName=Mobile Roughness}
   Specular=Multiply(A=Multiply(A=@n114, B=@n115), B=NamedRerouteUsage())
   Emissive Color
   Normal=Normalize(VectorInput=LinearInterpolate(A=@n118.PS, B=@n120, Alpha=@n121) {constAlpha=0.4})
   Opacity Mask=NamedRerouteUsage()
   Shading Model=ShadingModel()
   World Position Offset=NamedRerouteUsage()
  ) {attributeSetTypes [
    #1
    #3
    #2
    #10
    #6
    #5
    #9
    #8
   ]}
  NamedRerouteDeclaration(None=@n122) {name=MobileGrassV2, nodeColor=<rgba 1 0 0.1406 1>}
  n123 = Add(A=@n98.R, B=ScalarParameter() {defaultValue=-2000, group=GrassTypeSettings, parameterName=TerrainBlendOffset})
  n124 = ScalarParameter() {defaultValue=3000, group=GrassTypeSettings, parameterName=TerrainBlendFade}
  NamedRerouteDeclaration(None=Saturate(None=Divide(A=@n123, B=@n124))) {name=TerrainBlendMask, nodeColor=<rgba 0.3802 0.3802 0.3802 1>}
  n125 = NamedRerouteUsage()
  n126 = Multiply(A=Constant3Vector() {constant=<rgba -1 -1 1 0>}, B=NamedRerouteUsage())
  n127 = NamedRerouteUsage()
  n128 = LinearInterpolate(
   A=Transform(None=LinearInterpolate(A=@n126, Alpha=@n127))
   B=VertexNormalWS()
   Alpha=ScalarParameter() {group=GrassTypeSettings, parameterName=Grass_UseVertexNormalsInsteadOfNormalsTexture}
  )
  n129 = ScalarParameter() {defaultValue=0.15, group=GrassTypeSettings, parameterName=Grass_AmountOfNormalTypeNearCamera}
  NamedRerouteDeclaration(None=LinearInterpolate(A=@n125, B=LinearInterpolate(A=@n125, B=@n128, Alpha=@n129), Alpha=NamedRerouteUsage())) {name=Normal, nodeColor=<rgba 1 0 0.1406 1>}
  n130 = NamedRerouteDeclaration(None=LinearInterpolate(
    A=ScalarParameter() {defaultValue=0.16, group=GrassTypeSettings, parameterName=Specular_Near}
    B=ScalarParameter() {defaultValue=0.05, group=GrassTypeSettings, parameterName=Specular_Fade}
    Alpha=NamedRerouteUsage()
   )) {name=Specular_Fallback, nodeColor=<rgba 0.6641 0 1 1>}
  n131 = Multiply(A=NamedRerouteUsage(), B=@n130)
  n132 = NamedRerouteUsage()
  n133 = Multiply(A=Multiply(A=@n131, B=LinearInterpolate(Alpha=@n132) {constA=0.3, constB=1.2}), B=NamedRerouteUsage())
  n134 = NamedRerouteUsage()
  NamedRerouteDeclaration(None=ShadingPathSwitch(Default=QualitySwitch(Default=@n133, Low=@n134, High=@n134, Medium=@n134, Epic=@n133), Mobile=@n134) {desc=Migrated from FeatureLevelSwitch}) {name=Specular, nodeColor=<rgba 0 0.1562 1 1>}
  n135 = NamedRerouteDeclaration(None=LinearInterpolate(
    A=ScalarParameter() {defaultValue=0.75, group=GrassTypeSettings, parameterName=Roughness_Near, sliderMin=1}
    B=ScalarParameter() {defaultValue=2, group=GrassTypeSettings, parameterName=Roughness_Far}
    Alpha=NamedRerouteUsage()
   )) {name=Roughness_Fallback, nodeColor=<rgba 0.5 0 1 1>}
  n136 = Multiply(A=NamedRerouteUsage(), B=@n135)
  n137 = NamedRerouteUsage()
  NamedRerouteDeclaration(None=ShadingPathSwitch(Default=QualitySwitch(Default=@n136, Low=@n137, Medium=@n137), Mobile=@n137) {desc=Migrated from FeatureLevelSwitch}) {name=Roughness, nodeColor=<rgba 0 1 0.5078 1>}
  n138 = Multiply(
   A=LinearInterpolate(
    A=ScalarParameter() {defaultValue=-50, group=GrassTypeSettings_WPO, parameterName=SinkIntoGroundAmount}
    B=Constant()
    Alpha=PerInstanceFadeAmount()
   )
   B=Constant3Vector() {constant=<rgba 0 0 1 0>}
  )
  n139 = MaterialFunctionCall(Color=NamedRerouteUsage()) {materialFunction=&SplitComponents}
  n140 = CollectionParameter() {collection=&FortniteMaterialParameters, parameterName=GrassPushDownRadius}
  n141 = Divide(A=PixelDepth().R, B=@n140)
  n142 = Multiply(A=OneMinus(None=Saturate(None=@n141)), B=Constant() {r=-200})
  n143 = Constant()
  n144 = CollectionParameter() {collection=&FortniteMaterialParameters, parameterName=IsPlayerSliding}
  n145 = MaterialFunctionCall(X=@n139.R, Y=@n139.R, Z=Add(A=@n139.B, B=LinearInterpolate(A=@n142, B=@n143, Alpha=@n144))) {materialFunction=&MakeFloat3}
  n146 = StaticSwitchParameter(
   True=Constant3Vector()
   False=StaticSwitchParameter(True=Add(A=@n138, B=@n145.Result), False=@n145.Result) {group=GrassTypeSettings, parameterName=IsFallbackMaterial}
  ) {group=GrassTypeSettings_WPO, parameterName=DisableAllAnimation}
  n147 = NamedRerouteUsage()
  n148 = OneMinus(None=Subtract(A=Length(None=@n147), B=ScalarParameter() {defaultValue=65, parameterName=WPODist}))
  n149 = LinearInterpolate(A=Multiply(A=WorldPosition().XYZ) {constB=-1}, Alpha=Saturate(None=Sign(None=@n148))) {constB=0}
  NamedRerouteDeclaration(None=StaticSwitchParameter(True=Add(A=@n146, B=@n149), False=@n146) {parameterName=UseWPOSphereCull}) {name=WPO-out, nodeColor=<rgba 0 1 0.3438 1>}
  NamedRerouteDeclaration(None=@n25) {name=BaseColor, nodeColor=<rgba 0 1 0.5078 1>}
  NamedRerouteDeclaration(None=Constant()) {name=SSSColor, nodeColor=<rgba 1 0.0938 0 1>}
  n150 = ScalarParameter() {defaultValue=1, group=GrassTypeSettings, parameterName=SSSOpacity_Bottom}
  n151 = ScalarParameter() {defaultValue=0.5, group=GrassTypeSettings, parameterName=SSSOpacity_Top}
  n152 = NamedRerouteUsage()
  NamedRerouteDeclaration(None=Saturate(None=LinearInterpolate(A=@n150, B=@n151, Alpha=@n152))) {name=SSSOpacity, nodeColor=<rgba 0 0.3672 1 1>}
  n153 = VertexColor() {desc=Meshes using this material need the SimpleGrassPivotPainter.hda script run on them}
  NamedRerouteDeclaration(None=@n153.B) {name=B_VerticalGradient, desc="Vertical gradient - bottom=0 top=1", nodeColor=<rgba 1 0 0.0703 1>}
  NamedRerouteDeclaration(None=Frac(None=@n153.G)) {name=G_RandomPerBlade, desc=Random value per blade, nodeColor=<rgba 0.7109 0 1 1>}
  NamedRerouteDeclaration(None=Saturate(None=TwoSidedSign())) {name=2SidedSignMask, nodeColor=<rgba 1 0 1 1>}
  n154 = TextureSample() {
   samplerSource=SSM_Wrap_WorldGroupSettings
   samplerType=SAMPLERTYPE_Normal
   texture=/Game/Environments/Asteria/Foliage/Grass/Textures/Grass/CylinderNormal_N
  }
  NamedRerouteDeclaration(None=@n154.RGB) {name=CylinderNormal, nodeColor=<rgba 1 0.9375 0 1>}
  n155 = TextureSampleParameter2D() {
   group=GrassTypeSettings_Textures
   parameterName=Mask2DTexture
   samplerSource=SSM_Wrap_WorldGroupSettings
   samplerType=SAMPLERTYPE_Masks
   sortPriority=10
   texture=&T_WhiteMask
   channelNames {a="", b=RandomPerBlade, g=BakedAO, r=OpacityMask}
  }
  NamedRerouteDeclaration(None=@n155.RandomPerBlade) {name=TextureRandomPerBlade, nodeColor=<rgba 1 0.3984 0 1>}
  n156 = Normalize(VectorInput=CrossProduct(A=Constant3Vector() {constant=<rgba 0 0 1 0>}, B=NamedRerouteUsage()))
  n157 = Multiply(
   A=NamedRerouteUsage()
   B=LinearInterpolate(A=ScalarParameter() {defaultValue=0.5, group=GrassTypeSettings_WPO, parameterName=AnimBendFalloff}, Alpha=NamedRerouteUsage())
  )
  n158 = StaticSwitchParameter(True=NamedRerouteUsage(), False=NamedRerouteUsage()) {parameterName=UseCameraRelativeOptimization}
  n159 = StaticSwitchParameter(True=WorldPosition() {worldPositionShaderOffset=WPT_CameraRelativeNoOffsets}.XYZ, False=WorldPosition().XYZ) {parameterName=UseCameraRelativeOptimization}
  n160 = Subtract(A=Multiply(A=@n55, B=Divide(A=@n57, B=@n56)), B=@n53)
  n161 = NamedRerouteUsage()
  n162 = Add(A=Constant() {r=1}, B=ScalarParameter() {defaultValue=1, group=GrassTypeSettings_WPO, parameterName=Stiffness})
  n163 = StaticSwitchParameter(True=@n161, False=Power(Base=@n161, Exp=Clamp(None=@n162) {maxDefault=50, minDefault=1}) {constExponent=3}) {group=GrassTypeSettings, parameterName=IsFlowers}
  n164 = LinearInterpolate(A=ScalarParameter() {defaultValue=-1, group=GrassTypeSettings_WPO, parameterName=LowPolyCounterOffset}, Alpha=@n163)
  n165 = NamedRerouteUsage()
  n166 = NamedRerouteUsage()
  n167 = Multiply(
   A=Multiply(
    A=StaticSwitchParameter(True=@n164, False=@n163) {group=GrassTypeSettings_WPO, parameterName=UseLowPolyCounterOffset}
    B=Multiply(A=@n165, B=@n166)
   )
   B=ScalarParameter() {defaultValue=150, group=GrassTypeSettings_WPO, parameterName=Master_Anim_strength}
  )
  n168 = QualitySwitch(
   Default=StaticSwitchParameter(True=RotateAboutAxis(NormalizedRotationAxis=@n156, RotationAngle=@n157, PivotPoint=@n158, Position=@n159), False=Add(A=@n160, B=@n54)) {group=GrassTypeSettings_WPO, parameterName="UseRealRotationAnimation(MostExpensive)"}
   Low=@n167
   Medium=@n167
  )
  NamedRerouteDeclaration(None=NaniteReplace!(Default=@n167, Nanite=ShadingPathSwitch(Default=@n168, Mobile=@n167) {desc=Migrated from FeatureLevelSwitch})) {name=Simple Lookup Animation, nodeColor=<rgba 0 1 0.8359 1>}
  NamedRerouteDeclaration(None=VertexInterpolator(VS=@n97).PS) {name=PivotForRVT, nodeColor=<rgba 1 0.75 0 1>}
  NamedRerouteDeclaration(None=@n96.WindI_Direction_WithNoise) {name=WindDirectionWithNoise, nodeColor=<rgba 0 1 0.9297 1>}
  n169 = DotProduct(A=VertexNormalWS(), B=Normalize(VectorInput=WorldPosition() {worldPositionShaderOffset=WPT_CameraRelative}.XYZ))
  n170 = Multiply(A=OneMinus(None=Abs(None=@n169)), B=ScalarParameter() {defaultValue=4, group=CameraFacingControls, parameterName=Thickening})
  NamedRerouteDeclaration(None=LinearInterpolate(
    A=MaterialFunctionCall(WidthBase=@n170, WidthTip=@n170) {materialFunction=&SplineThicken}.WorldPositionOffset
    B=Constant3Vector()
    Alpha=NamedRerouteUsage()
   )) {name=RotateGrasstoCamera, nodeColor=<rgba 0.5469 1 0 1>}
  n171 = ComponentMask(None=NamedRerouteUsage()) {b=true}
  n172 = OneMinus(None=Power(Base=Abs(None=@n171), Exp=ScalarParameter() {defaultValue=1.25, group=GrassTypeSettings, parameterName=GrassV2_Slope_Pow}))
  n173 = ScalarParameter() {defaultValue=1, group=GrassTypeSettings, parameterName=GrassV2_Slope_Mod}
  n174 = MaterialFunctionCall!(Grass_UseCustomSlopeMask=StaticBool() {value=true}, Grass_CustomSlopeMask=Saturate(None=Multiply(A=@n172, B=@n173))) {materialFunction=/Game/Athena/Helios/Environments/Landscape/Materials/MaterialFunctions_Utilities/MF_Helios_StaticMeshLandscapeLayerMatch}
  n175 = VectorParameter() {group=GrassTypeSettings, parameterName=GrassColor_NoGrassBiomeMask, defaultValue=<rgba 0.0471 0.149 0.0471 1>}
  n176 = VectorParameter() {group=GrassTypeSettings, parameterName=GrassColor_NoGrassBiomeMask_Mobile, defaultValue=<rgba 0.0471 0.149 0.0471 1>}
  n177 = StaticSwitchParameter(
   True=GetMaterialAttributes(None=@n174.GrassLayer) {attributeGetTypes [#1]}.Base Color
   False=ShadingPathSwitch(Default=@n175.RGB, Mobile=@n176.RGB) {desc=Migrated from FeatureLevelSwitch}
  ) {defaultValue=true, desc=False will use a simple color vector, group=GrassTypeSettings, parameterName=UseGrassBiomeMaskColors}
  NamedRerouteDeclaration(None=@n177) {name=LandscapeColor_Influence, nodeColor=<rgba 1 0.9375 0 1>}
  n178 = StaticSwitchParameter(True=@n92.A, False=@n155.OpacityMask) {
   desc=False uses the Mask2DTexture Red channel. This should be false unless absolutely necessary.
   group=GrassTypeSettings
   parameterName=UseDiffuseAlphaAsOpacityMask
  }
  NamedRerouteDeclaration(None=StaticSwitchParameter(True=@n178, False=Constant()) {group=GrassTypeSettings, parameterName=IsFallbackMaterial}) {name=Opacity Mask, nodeColor=<rgba 1 0.3984 0 1>}
  NamedRerouteDeclaration(None=@n155.BakedAO) {name=BakedAO, nodeColor=<rgba 1 0.3984 0 1>}
  n179 = ScalarParameter() {group=GrassTypeSettings, parameterName=BakedAOintensity}
  n180 = NamedRerouteUsage()
  n181 = NamedRerouteUsage()
  n182 = Multiply(A=Saturate(None=LinearInterpolate(A=@n179, Alpha=@n180)), B=@n181)
  n183 = ShadingPathSwitch(
   Default=QualitySwitch(Default=StaticSwitchParameter(True=@n182, False=@n181) {group=GrassTypeSettings, parameterName=IsFallbackMaterial}, Low=@n181, Medium=@n181)
   Mobile=@n181
  ) {desc=Migrated from FeatureLevelSwitch}
  n184 = StaticSwitchParameter(True=LinearInterpolate(A=@n183, B=NamedRerouteUsage(), Alpha=NamedRerouteUsage()), False=@n183) {group=GrassTypeSettings, parameterName=FadeToGrassBiomeColorInDistance}
  n185 = ScalarParameter() {defaultValue=0.828, group=GrassTypeSettings, parameterName=MobileDiffuseBrightness}
  NamedRerouteDeclaration(None=ShadingPathSwitch(Default=@n184, Mobile=Multiply(A=@n184, B=@n185)) {desc=Migrated from FeatureLevelSwitch}) {name=Final BaseColor, nodeColor=<rgba 1 0 0.8203 1>}
  NamedRerouteDeclaration(None=@n41) {name=PlayerSphereMask, nodeColor=<rgba 0 0.8828 1 1>}
  NamedRerouteDeclaration(None=@n40) {name=PlayerPushDirection, nodeColor=<rgba 0 0.6016 1 1>}
  n186 = SamplePhysicsVectorField!(World Position=NamedRerouteUsage()) {fieldTarget=Vector_LinearVelocity}
  n187 = MaterialFunctionCall(Vector 3=Add(A=@n186) {constB=0.0001}) {materialFunction=/Engine/Functions/Engine_MaterialFunctions02/Utility/VectorLength}
  n188 = Constant()
  NamedRerouteDeclaration(None=ShadingPathSwitch(Default=Saturate(None=@n187.V3 Length), Mobile=@n188) {desc=Migrated from FeatureLevelSwitch}) {name=PhysicsFieldMask, nodeColor=<rgba 0 1 0.3906 1>}
  n189 = Constant3Vector() {constant=<rgba 0 0 1 0>}
  NamedRerouteDeclaration(None=ShadingPathSwitch(
    Default=MaterialFunctionCall(Vector=@n186, Default=@n189) {materialFunction=&SafeNormalize}.Result
    Mobile=Constant3Vector() {constant=<rgba 0 0 1 0>}
   ) {desc=Migrated from FeatureLevelSwitch}) {name=PhysicsFieldDirection, nodeColor=<rgba 0.3828 0 1 1>}
  n190 = Min(A=@n187.V3 Length, B=ScalarParameter() {defaultValue=1, group=GrassTypeSettings_WPO, parameterName=MaxRotation_PhysicsField_Clamp})
  n191 = Time()
  n192 = Multiply(A=ScalarParameter() {defaultValue=351, group=GrassTypeSettings_WPO, parameterName=PhysicsField_NoiseOffset}, B=NamedRerouteUsage())
  n193 = Frac(None=Multiply(
    A=ScalarParameter() {defaultValue=2.5, group=GrassTypeSettings_WPO, parameterName=MaxRotation_PhysicsField_NoiseSpeed}
    B=Add(A=@n191, B=@n192)
   ))
  n194 = OneMinus(None=@n193)
  n195 = LinearInterpolate(
   A=Multiply(A=Multiply(A=@n193, B=@n194)) {constB=2}
   Alpha=ScalarParameter() {defaultValue=0.6, group=GrassTypeSettings_WPO, parameterName=MaxRotation_PhysicsField_MInBend}
  )
  NamedRerouteDeclaration(None=ShadingPathSwitch(Default=Multiply(A=@n190, B=@n195), Mobile=@n188) {desc=Migrated from FeatureLevelSwitch}) {name=PhysicsFieldStrength, nodeColor=<rgba 1 0 0.2109 1>}
  n196 = TextureSampleParameter2D() {
   group=GrassTypeSettings_Textures
   parameterName=SMR_Texture
   samplerSource=SSM_Wrap_WorldGroupSettings
   samplerType=SAMPLERTYPE_Masks
   sortPriority=15
   texture=&T_WhiteMask
   channelNames {a="", b=Roughness, g=ColorSystemMask, r=Specular}
  }
  NamedRerouteDeclaration(None=@n196.Specular) {name=SpecularTexture, nodeColor=<rgba 1 0 0.1309 1>}
  NamedRerouteDeclaration(None=@n196.Roughness) {name=Roughness Texture, nodeColor=<rgba 1 0 0.1309 1>}
  NamedRerouteDeclaration(None=@n196.ColorSystemMask) {name=ColorSytemTexturemask, nodeColor=<rgba 1 0 0.1309 1>}
  NamedRerouteDeclaration(None=TransformPosition(None=Constant3Vector()) {transformSourceType=TRANSFORMPOSSOURCE_Instance, transformType=TRANSFORMPOSSOURCE_World}) {name=InstancePosition, nodeColor=<rgba 0.3828 1 0 1>}
  NamedRerouteDeclaration(None=@n153.R) {name=VertexColor_Red, nodeColor=<rgba 0 0.7187 1 1>}
  n197 = StaticSwitchParameter(True=NamedRerouteUsage(), False=NamedRerouteUsage()) {group=GrassTypeSettings, parameterName=IsFallbackMaterial}
  n198 = Multiply(A=ScalarParameter() {group=GrassTypeSettings, parameterName=HueVarPerBlade}, B=Subtract(A=@n197) {constB=0.5})
  n199 = PerInstanceRandom()
  n200 = Add(
   A=@n198
   B=Multiply(A=Subtract(A=@n199) {constB=0.5}, B=ScalarParameter() {defaultValue=-0.024, group=Color, parameterName=HueVarPerCluster})
  )
  n201 = Multiply(
   A=NamedRerouteUsage()
   B=VectorParameter() {group=GrassTypeSettings, parameterName=ColorTint, sortPriority=-500, defaultValue=<rgba 1 1 1 1>}.RGB
  )
  NamedRerouteDeclaration(None=StaticSwitchParameter(
    True=MaterialFunctionCall(Hue Shift Percentage=@n200, Texture=@n201) {materialFunction=&HueShift}.Result
    False=NamedRerouteUsage()
   ) {group=GrassTypeSettings, parameterName=IsFlowers}) {name=ColorLowQuality, nodeColor=<rgba 0.1953 0 1 1>}
  n202 = RuntimeVirtualTextureSampleParameter!(UVs, World Position=NamedRerouteUsage(), Mip Level=Constant()) {
   bEnableFeedback=false
   group=RVT
   materialType=BaseColor_Normal_Specular_YCoCg
   mipValueMode=RVTMVM_MipLevel
   parameterName=RVT Sample Target
   virtualTexture=/Game/Athena/Apollo/Maps/Landscape/VT/RVT_Apollo_Landscape
  }
  n203 = NamedRerouteUsage()
  n204 = MaterialFunctionCall(NoRVT=@n203, RVT=StaticSwitchParameter(True=@n202.BaseColor, False=@n203) {defaultValue=true, group=GrassTypeSettings, parameterName=UseRVT}) {materialFunction=&FN_RVT_Switch_Material}
  NamedRerouteDeclaration(None=ShadingPathSwitch(Default=QualitySwitch(Default=@n204.Result, Low=@n203), Mobile=@n203) {desc=Migrated from FeatureLevelSwitch}) {name=RVTorBiomeColor, nodeColor=<rgba 1 0.5391 0 1>}
  n205 = NamedRerouteUsage()
  n206 = NamedRerouteUsage()
  n207 = NamedRerouteUsage()
  NamedRerouteDeclaration(None=LinearInterpolate(A=LinearInterpolate(A=@n205, B=@n206, Alpha=@n207), B=NamedRerouteUsage(), Alpha=NamedRerouteUsage())) {name=BlendedDirectionVector, nodeColor=<rgba 0 1 0.3203 1>}
  n208 = NamedRerouteUsage()
  n209 = ScalarParameter() {defaultValue=0.38, group=GrassTypeSettings_WPO, parameterName=MaxRotation_texAnim}
  n210 = NamedRerouteUsage()
  n211 = ScalarParameter() {defaultValue=1, group=GrassTypeSettings_WPO, parameterName=MaxRotation_PhysicsField}
  n212 = LinearInterpolate(
   A=LinearInterpolate(A=Multiply(A=@n208, B=@n209), B=Multiply(A=@n210, B=@n211), Alpha=NamedRerouteUsage())
   B=ScalarParameter() {defaultValue=0.375, group=GrassTypeSettings_WPO, parameterName=MaxRotation_PlayerPush}
   Alpha=NamedRerouteUsage()
  )
  NamedRerouteDeclaration(None=@n212) {name=AnimStrength, nodeColor=<rgba 1 0.3047 0 1>}
  n213 = NamedRerouteUsage()
  n214 = NamedRerouteUsage()
  n215 = ScalarParameter() {defaultValue=0.85, group=GrassTypeSettings, parameterName=GrassV2_Flat_Roughness}
  n216 = ScalarParameter() {defaultValue=-0.25, group=GrassTypeSettings, parameterName=GrassV2_Noise3_A}
  n217 = ScalarParameter() {defaultValue=1.5, group=GrassTypeSettings, parameterName=GrassV2_Noise3_B}
  n218 = MaterialFunctionCall!(Coordinates, R_01_Scale, R_02_Scale, R_03_Scale, R_04_Scale, G_01_Scale, G_02_Scale, G_03_Scale, G_04_Scale) {materialFunction=/Game/Athena/Helios/Environments/Landscape/Materials/MaterialFunctions_Utilities/MF_VariationMasks2}
  n219 = Multiply(A=Saturate(None=LinearInterpolate(A=@n216, B=@n217, Alpha=@n218.G_03)), B=Constant() {r=0.15})
  n220 = SetMaterialAttributes!(
   MaterialAttributes
   Base Color=StaticSwitchParameter(True=Multiply(A=@n213, B=@n214), False=@n214) {parameterName=UseDiffuseColor}
   Specular=NamedRerouteUsage()
   Roughness=Saturate(None=Add(A=@n215, B=@n219))
   Opacity=NamedRerouteUsage()
   Opacity Mask=NamedRerouteUsage()
   Normal=NamedRerouteUsage()
   Subsurface Color=Constant()
   World Position Offset=NamedRerouteUsage()
   Shading Model=ShadingModel() {shadingModel=MSM_Subsurface}
  ) {attributeSetTypes [
    #1
    #2
    #3
    #4
    #5
    #6
    #7
    #8
    #9
   ]}
  NamedRerouteDeclaration(None=ShadingPathSwitch(Default=@n220, Mobile=NamedRerouteUsage())) {name=GrassV2, nodeColor=<rgba 1 0 0.1406 1>}
  NamedRerouteDeclaration(None=@n61.Result) {name=PivotPosition_LocalSpace, nodeColor=<rgba 0 0.9766 1 1>}
  n221 = NamedRerouteUsage()
  n222 = Multiply(A=@n221) {constB=2.5}
  NamedRerouteDeclaration(None=Saturate(None=Add(A=@n222) {constB=-0.9})) {name=Height Mask, nodeColor=<rgba 0.9453 1 0 1>}
  n223 = PerInstanceRandom()
  n224 = Multiply(A=Add(A=NamedRerouteUsage(), B=VertexInterpolator(VS=@n223).PS)) {constB=2.5564}
  n225 = Subtract(A=Multiply(A=NamedRerouteUsage()) {constB=3}) {constB=0.2}
  n226 = VertexInterpolator(VS=Add(A=Saturate(None=@n225)) {constB=0.2})
  NamedRerouteDeclaration(None=Multiply(A=Frac(None=@n224), B=@n226.PS)) {name=RandomGrayscale, nodeColor=<rgba 1 0 0.7031 1>}
  NamedRerouteDeclaration(None=@n226.PS) {name=Wind Mask, nodeColor=<rgba 0 1 0.8125 1>}
  n227 = Multiply(
   A=Multiply(
    A=Power(Base=@n221) {constExponent=16}
    B=ScalarParameter() {defaultValue=2, group=GrassTypeSettings_Mobile, parameterName=Mobile Bright Tips}
   ) {constB=1.5}
   B=NamedRerouteUsage()
  )
  n228 = OneMinus(None=NamedRerouteUsage())
  NamedRerouteDeclaration(None=Add(A=Multiply(A=@n227, B=@n228))) {name=Bright Tips, nodeColor=<rgba 0 0.0156 1 1>}
  NamedRerouteUsage()
  NamedRerouteUsage()
  NamedRerouteUsage()
  NamedRerouteUsage()
  NamedRerouteUsage()
  NamedRerouteUsage()
  NamedRerouteUsage()
  NamedRerouteUsage()
  NamedRerouteUsage()
  NamedRerouteUsage()
  NaniteReplace()
  n229 = ShadingModel() {shadingModel=MSM_Subsurface}
  n230 = Transform(None=NamedRerouteUsage()) {transformSourceType=TRANSFORMSOURCE_Instance}
  n231 = NamedRerouteUsage()
  n232 = Multiply(
   A=TwoSidedSign()
   B=LinearInterpolate(A=Transform(None=@n231), B=VertexNormalWS(), Alpha=ScalarParameter() {defaultValue=0.5, parameterName=DebugNormalttangent13})
  )
  n233 = LinearInterpolate(A=@n232, B=@n230, Alpha=ScalarParameter() {defaultValue=0.5, parameterName=DebugNormalt3234})
  n234 = Multiply(
   A=Power(Base=NamedRerouteUsage(), Exp=ScalarParameter() {defaultValue=1.5, parameterName=debug-rough-noise-pow1})
   B=ScalarParameter() {defaultValue=3, parameterName=debug-rough-noise-mod}
  )
  n235 = DotProduct(A=Multiply(A=TwoSidedSign(), B=VertexNormalWS()), B=SkyAtmosphereLightDirection())
  n236 = Multiply(A=Max(A=Saturate(None=@n234), B=PerInstanceRandom()), B=@n235)
  n237 = VectorParameter() {parameterName=G2_C5, defaultValue=<rgba 0.0454 0.1 0.012 1>}
  n238 = VectorParameter() {parameterName=G2_C1, defaultValue=<rgba 0.0763 0.1333 0.0063 1>}
  n239 = VectorParameter() {parameterName=G2_C2, defaultValue=<rgba 0.0588 0.0875 0.0101 1>}
  n240 = MaterialFunctionCall() {materialFunction=&MF_VariationMasks2}
  n241 = Saturate(None=LinearInterpolate(A=Constant(), B=Constant() {r=1.5}, Alpha=@n240.G_01))
  n242 = LinearInterpolate(
   A=ScalarParameter() {defaultValue=1.5, group=Grass, parameterName=Grass2_Noise2_A}
   B=ScalarParameter() {defaultValue=-9, group=Grass, parameterName=Grass2_Noise2_B}
   Alpha=@n240.G_02
  )
  n243 = LinearInterpolate(
   A=ScalarParameter() {defaultValue=-0.25, group=Grass, parameterName=Grass2_Noise3_A}
   B=ScalarParameter() {defaultValue=1.5, group=Grass, parameterName=Grass2_Noise3_B}
   Alpha=@n240.G_03
  )
  n244 = LinearInterpolate(
   A=LinearInterpolate(
    A=LinearInterpolate(A=@n238.RGB, B=@n239.RGB, Alpha=@n241)
    B=VectorParameter() {parameterName=G2_C3, defaultValue=<rgba 0.0776 0.1333 0.0191 1>}.RGB
    Alpha=Saturate(None=@n242)
   )
   B=VectorParameter() {parameterName=G2_C4, defaultValue=<rgba 0.114 0.1108 0.0191 1>}.RGB
   Alpha=Multiply(A=Saturate(None=@n243), B=ScalarParameter() {defaultValue=0.65, group=Grass, parameterName=Grass2_Noise3_Mod})
  )
  n245 = Saturate(None=Multiply(A=@n240.G_04, B=ScalarParameter() {defaultValue=4, group=Grass, parameterName=Grass2_Noise4_Mod}))
  n246 = Saturate(None=Multiply(
    A=LinearInterpolate(A=@n237.RGB, B=@n244, Alpha=@n245)
    B=VectorParameter() {group=Grass, parameterName=Grass2_ModColorTemp, defaultValue=<rgba 1.1081 1.3 1.047 1>}.RGB
   ))
  n247 = ScalarParameter() {defaultValue=0.1, parameterName=debugsscolormod44}
  n248 = GetMaterialAttributes(None=@n10.Result) {attributeGetTypes [#2, #8]}
  n249 = SetMaterialAttributes!(
   MaterialAttributes=@n10.Result
   World Position Offset=@n248.World Position Offset
   Shading Model=@n229
   Normal=LinearInterpolate(
    A=@n230
    B=LinearInterpolate(A=@n233, B=@n232, Alpha=@n236)
    Alpha=ScalarParameter() {defaultValue=1, desc=override, parameterName=DebugNormalt3234_OR}
   )
   Subsurface Color=LinearInterpolate(A=Multiply(A=@n246, B=@n247), B=Constant(), Alpha=@n235)
   Opacity=LinearInterpolate(
    A=ScalarParameter() {defaultValue=1, parameterName=SSopacity_debug444_A}
    B=ScalarParameter() {defaultValue=0.75, parameterName=SSopacity_debug444_B}
    Alpha=NamedRerouteUsage()
   )
   Roughness=LinearInterpolate(
    A=ScalarParameter() {defaultValue=1, parameterName=debug-rough-LV-zero}
    B=ScalarParameter() {defaultValue=0.6, parameterName=debug-rough-LV-one}
    Alpha=@n235
   )
   Specular=@n248.Specular
   Base Color=@n246
  ) {attributeSetTypes [
    #8
    #9
    #6
    #7
    #4
    #3
    #2
    #1
   ]}
  Reroute(None=@n249)
  n250 = Multiply(
   A=NamedRerouteUsage()
   B=LinearInterpolate(A=ScalarParameter() {defaultValue=-0.0832, group=GrassTypeSettings, parameterName=ValueVar}, Alpha=@n197)
  )
  n251 = StaticSwitchParameter(
   True=@n250
   False=MaterialFunctionCall(Hue Shift Percentage=@n198, Texture=@n250) {materialFunction=&HueShift}.Result
  ) {group=GrassTypeSettings, parameterName=IsFlowers}
  n252 = NamedRerouteUsage()
  Reroute(None=ShadingPathSwitch(Default=QualitySwitch(Default=@n251, Low=@n252, Medium=@n252), Mobile=@n252) {desc=Migrated from FeatureLevelSwitch})
  Reroute(None=@n10.Result)
  Reroute(None=@n203)
  n253 = Constant3Vector() {constant=<rgba 0 0 1 0>}
  n254 = NamedRerouteUsage()
  n255 = ScalarParameter() {defaultValue=0.5, group=GrassTypeSettings_WPO, parameterName=AnimBendFalloff}
  n256 = NamedRerouteUsage()
  RotateAboutAxis(
   NormalizedRotationAxis=Normalize(VectorInput=CrossProduct(A=@n253, B=@n254))
   RotationAngle=Multiply(A=NamedRerouteUsage(), B=LinearInterpolate(A=@n255, Alpha=@n256))
   PivotPoint=NamedRerouteUsage()
   Position=LocalPosition().XYZ
  )
  n257 = ScalarParameter() {defaultValue=-3000, group=GrassTypeSettings, parameterName=FlowerFadeOffset}
  Saturate(None=Divide(A=Add(A=@n98.R, B=@n257), B=ScalarParameter() {defaultValue=5000, group=GrassTypeSettings, parameterName=FlowerFarFade}))
  ScalarParameter() {defaultValue=0.3, group=GrassTypeSettings_SB, parameterName=SB_StemNormalAmt}
  ScalarParameter() {defaultValue=0.75, group=GrassTypeSettings_SB, parameterName=SB_FlowerNormalAmt}
  ScalarParameter() {defaultValue=0.1, group=GrassTypeSettings_SB, parameterName=SB_BlueNormalAmt}
  ScalarParameter() {group=GrassTypeSettings, parameterName=Specular_Far}
  n258 = GetMaterialAttributes() {attributeGetTypes [#1, #6]}
  SetMaterialAttributes!(MaterialAttributes=@n258.MaterialAttributes, Base Color=NamedRerouteUsage(), Normal=@n258.MaterialAttributes) {attributeSetTypes [#1, #6]}
  SetMaterialAttributes!(MaterialAttributes=@n10.Result, World Position Offset=Constant()) {attributeSetTypes [#8]}
  SetMaterialAttributes!(MaterialAttributes, World Position Offset=Constant(), Shading Model=@n229) {attributeSetTypes [#8, #9]}
  SphereMask(A=CameraPositionWS(), B=NamedRerouteUsage(), Radius=@n140, Hardness=Constant())
  n259 = LinearInterpolate(
   A=Multiply(A=NamedRerouteUsage(), B=ScalarParameter() {group=GrassTypeSettings, parameterName=SSSColor})
   B=Constant()
   Alpha=NamedRerouteUsage()
  )
  n260 = NamedRerouteUsage()
  n261 = ScalarParameter() {defaultValue=0.5, group=GrassTypeSettings, parameterName=SSSBrightnessBase}
  StaticSwitchParameter(True=Multiply(A=@n259, B=@n260), False=Multiply(A=@n259, B=LinearInterpolate(A=@n261, Alpha=@n260))) {group=GrassTypeSettings, parameterName=IsFallbackMaterial}
  StaticSwitchParameter(True=@n11, False=NamedRerouteUsage()) {group=GrassTypeSettings, parameterName=IsFallbackMaterial}
  StaticSwitchParameter(True=NamedRerouteUsage()) {group=GrassTypeSettings_MaskedColorVariation, parameterName=UseMaskedColorVariation}
  StaticSwitchParameter(True=@n76, False=NamedRerouteUsage()) {defaultValue=true, group=GrassTypeSettings_SB, parameterName=SB_UseGChannelForOpacityMask}
  StaticSwitchParameter(
   True=MaterialFunctionCall!(Scale, Camera Relative World Position=NamedRerouteUsage()) {materialFunction=/Engine/Functions/Engine_MaterialFunctions03/Texturing/PeriodicWorldPosition}.XYZ
   False=@n93
  ) {parameterName=UseCameraRelativeOptimization}
  StaticSwitchParameter(True=Add(A=NamedRerouteUsage())) {group=CameraFacingControls, parameterName=UseCameraFacing SplineThicken}
  n262 = NamedRerouteUsage()
  n263 = VertexNormalWS()
  StaticSwitchParameter(True=StaticSwitchParameter(True=Transform(None=@n262), False=@n263) {group=GrassTypeSettings, parameterName=IsFallbackMaterial}, False=@n263) {group=GrassTypeSettings, parameterName=UseNormalTexture}
  n264 = NamedRerouteUsage()
  StaticSwitchParameter(True=@n264, False=NamedRerouteUsage()) {group=GrassTypeSettings, parameterName=UseTextures}
  StaticSwitchParameter(True=@n131, False=@n130) {group=GrassTypeSettings, parameterName=UseTextures}
  StaticSwitchParameter(True=@n136, False=@n135) {group=GrassTypeSettings, parameterName=UseTextures}
  n265 = NamedRerouteUsage()
  n266 = Constant()
  n267 = TextureSample(UVs=Add(A=TextureCoordinate() {coordinateIndex=2}, B=AppendVector(A=@n212, B=@n266))) {
   samplerSource=SSM_Wrap_WorldGroupSettings
   samplerType=SAMPLERTYPE_LinearColor
   texture=/Game/Environments/Asteria/Foliage/Grass/MasterMaterial/GenericShared_Grass_Foliage_Anim
  }
  n268 = Multiply(A=@n267.G, B=ScalarParameter() {defaultValue=1, group=GrassTypeSettings_WPO, parameterName=Z_Offset_Mult})
  n269 = Constant3Vector() {constant=<rgba 0 0 -1 0>}
  StaticSwitchParameter(True=Add(A=Multiply(A=@n265, B=@n267.R), B=Multiply(A=@n268, B=@n269)), False=@n167) {group=GrassTypeSettings_WPO, parameterName=UseHighQualityRotationAnimation}
  n270 = NamedRerouteUsage()
  StaticSwitchParameter(
   True=LinearInterpolate(A=ScalarParameter() {defaultValue=-1, group=GrassTypeSettings_WPO, parameterName=LowPolyCounterOffset}, Alpha=@n270)
   False=@n270
  ) {group=GrassTypeSettings_WPO, parameterName=UseLowPolyCounterOffset}
  n271 = AppendVector(A=Constant2Vector() {g=1, r=1}, B=TwoSidedSign())
  n272 = Add(
   A=Multiply(A=NamedRerouteUsage(), B=ScalarParameter() {defaultValue=4, group=GrassTypeSettings, parameterName=RVTNormalInfluence})
   B=Transform(None=Multiply(A=@n264, B=@n271))
  )
  n273 = LinearInterpolate(A=@n272, B=NamedRerouteUsage(), Alpha=NamedRerouteUsage())
  StaticSwitchParameter(
   True=MaterialFunctionCall(BaseNormal=NamedRerouteUsage(), AdditionalNormal=NamedRerouteUsage()) {materialFunction=&BlendAngleCorrectedNormals}.Result
   False=Normalize(VectorInput=StaticSwitchParameter(True=@n273, False=@n272) {group=GrassTypeSettings, parameterName=FadeToGrassBiomeColorInDistance})
  ) {group=GrassTypeSettings, parameterName=IsFallbackMaterial}
  Transform(None=NamedRerouteUsage())
 ]
}
